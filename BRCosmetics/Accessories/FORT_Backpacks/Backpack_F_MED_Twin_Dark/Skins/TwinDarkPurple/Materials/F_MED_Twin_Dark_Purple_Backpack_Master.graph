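# Master material for the purple Twin Dark backpack cosmetic, adding animated swirl, bubble and rim glow to the standard character shading stack.
# FX texture channels blend the MF_TwinlightPurple pattern, bubble layer and purple Fresnel into emissive, behind a static switch, and adjust base color and roughness.
# Bubbles combine two scrolling world-aligned projections on pre-skinned position with panned view-aligned reflection noise.
# A quality switch zeroes part of the bubble layer on Low and Medium.
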Material /BRCosmetics/Accessories/FORT_Backpacks/Backpack_F_MED_Twin_Dark/Skins/TwinDarkPurple/Materials/F_MED_Twin_Dark_Purple_Backpack_Master
graph {
 properties {
  bAutomaticallySetUsageInEditor=false
  bUseMaterialAttributes=true
  bUsedWithClothing=true
  bUsedWithMeshDeformer=true
  bUsedWithMorphTargets=true
  bUsedWithNanite=true
  bUsedWithSkeletalMesh=true
  pixelDepthOffsetMode=PDOM_Legacy
  refractionMethod=RM_IndexOfRefraction
  shadingModel=MSM_FromMaterialExpression
  subsurfaceProfile=&SS_HeroSkin_02
  usedShadingModels=Default Lit | Subsurface | Clear Coat | Subsurface Profile | Cloth | Eye
 }
 nodes [
  n1 = MaterialFunctionCall() {materialFunction=&BaseFN_CharacterFunction_PreFX}
  n2 = TextureSample() {
   samplerSource=SSM_Wrap_WorldGroupSettings
   samplerType=SAMPLERTYPE_Masks
   texture=/BRCosmetics/Accessories/FORT_Backpacks/Backpack_F_MED_Twin_Dark/Skins/TwinDarkPurple/Textures/T_F_MED_Twin_Dark_Purple_Backpack_M
  }
  n3 = TextureSample() {
   samplerSource=SSM_Wrap_WorldGroupSettings
   texture=/BRCosmetics/Accessories/FORT_Backpacks/Backpack_F_MED_Twin_Dark/Skins/TwinDarkPurple/Textures/T_F_MED_Twin_Dark_Purple_Backpack_D
  }
  n4 = TextureSample() {
   samplerSource=SSM_Wrap_WorldGroupSettings
   texture=/BRCosmetics/Accessories/FORT_Backpacks/Backpack_F_MED_Twin_Dark/Skins/TwinDarkPurple/Textures/T_F_MED_Twin_Dark_Purple_Backpack_FX
  }
  n5 = TextureSample() {
   samplerSource=SSM_Wrap_WorldGroupSettings
   samplerType=SAMPLERTYPE_Masks
   texture=/BRCosmetics/Characters/Player/Female/Medium/Bodies/F_MED_Twin_Dark/Skins/Twin_Light_Purple/Textures/T_Twin_Light_Purple_Body_S
  }
  n6 = Multiply(A=@n4.B, B=Constant() {r=8}) {constB=2}
  n7 = MaterialFunctionCall() {materialFunction=/BRCosmetics/Characters/Player/Female/Medium/Bodies/F_MED_Twin_Dark/Skins/Twin_Light_Purple/Materials/MF_TwinlightPurple}
  n8 = LinearInterpolate(A=LinearInterpolate(A=NamedRerouteUsage(), B=@n7.Result, Alpha=Saturate(None=@n6)), B=NamedRerouteUsage(), Alpha=@n4.R)
  n9 = Fresnel(ExponentIn=Constant() {r=2})
  n10 = Add(A=Multiply(A=Power(Base=@n9), B=Constant3Vector() {constant=<rgba 0.1367 0.0086 0.2831 0>}), B=@n8)
  n11 = SetMaterialAttributes!(
   MaterialAttributes
   Base Color=LinearInterpolate(A=NamedRerouteUsage(), B=Multiply(A=@n2.R, B=@n3.RGB), Alpha=@n4.R)
   Curvature or Metallic=NamedRerouteUsage()
   Specular=NamedRerouteUsage()
   Roughness=LinearInterpolate(A=NamedRerouteUsage(), B=LinearInterpolate(Alpha=@n5.B) {constA=0.2, constB=0}, Alpha=@n4.R)
   Emissive Color=StaticSwitchParameter(True=LinearInterpolate(A=@n8, B=@n10, Alpha=@n4.G), False=NamedRerouteUsage()) {defaultValue=true, parameterName=use swirl pattern body}
   Normal=NamedRerouteUsage()
  ) {attributeSetTypes [
    #1
    #2
    #3
    #4
    #5
    #6
   ]}
  n12 = MaterialFunctionCall(AO=@n1.AO, Distance Blend=@n1.AO, Fresnel=@n1.AO, Gmap=@n1.AO, Pre Skinned Local Position=@n1.AO, Material=@n11) {materialFunction=&BaseFN_CharacterFunction_PostFX}
  MaterialFunctionCall(In=@n12.Result) {materialFunction=&CINE_MF_CharPostFX}
  n13 = Multiply(
   A=MaterialFunctionCall!(ReflectionVector, ReflectionTexture) {materialFunction=/Engine/Functions/Engine_MaterialFunctions01/Reflections/ViewAlignedReflection}.UVs
   B=ScalarParameter() {defaultValue=1, group=Slurp, parameterName=Bubble_ViewDependantDistortion}
  )
  n14 = Time()
  n15 = AppendVector(
   A=ScalarParameter() {defaultValue=0.025, group=Slurp, parameterName=Bubble_ViewDistortion_SpeedX}
   B=ScalarParameter() {defaultValue=0.05, group=Slurp, parameterName=Bubble_ViewDistortion_SpeedY}
  )
  n16 = TextureSample(UVs=Rotator(Coordinate=Panner(Coordinate=@n13, Time=@n14, Speed=@n15), Time=@n14) {speed=0.0001}) {
   constMipValue=0
   texture=/BRCosmetics/Characters/Player/Female/Medium/Bodies/F_MED_Twin_Dark/Skins/Twin_Light_Purple/Textures/T_twinlightPurpleBlurredNoise
  }
  n17 = TextureObject() {texture=/BRCosmetics/Characters/Player/Female/Medium/Bodies/F_MED_Twin_Dark/Skins/Twin_Light_Purple/Textures/T_twinlightBubble}
  n18 = Constant3Vector() {constant=<rgba 200 200 200 1>}
  n19 = Constant() {r=0.025}
  n20 = VertexInterpolator(VS=PreSkinnedPosition())
  n21 = Constant3Vector() {constant=<rgba 0.66 1 2 1>}
  n22 = Multiply(A=Multiply(A=Time(), B=@n21), B=Constant() {r=0.3})
  n23 = ComponentMask(None=MaterialFunctionCall(TextureObject=@n17, TextureSize=Multiply(A=@n18, B=@n19), WorldPosition=Subtract(A=@n20.PS, B=@n22)) {materialFunction=/Game/Weapons/WeaponSkins/Wraps/MaterialFunctions/WorldAlignedTexture}.XYZ Texture) {r=true}
  n24 = Constant3Vector() {constant=<rgba 200 200 200 1>}
  n25 = Constant() {r=0.04}
  n26 = VertexInterpolator(VS=PreSkinnedPosition())
  n27 = Multiply(A=Multiply(A=Time(), B=@n21), B=Constant() {r=0.5})
  n28 = ComponentMask(None=MaterialFunctionCall(TextureObject=@n17, TextureSize=Multiply(A=@n24, B=@n25), WorldPosition=Subtract(A=@n26.PS, B=@n27)) {materialFunction=&WorldAlignedTexture}.XYZ Texture) {g=true}
  n29 = Desaturation(None=Multiply(A=Max(A=@n23, B=@n28), B=ScalarParameter() {defaultValue=0.1}), Fraction=Constant() {r=1})
  n30 = Constant3Vector() {constant=<rgba 0.52 0.2584 0.52 0>}
  NamedRerouteDeclaration(None=Multiply(A=Power(Base=@n16.R), B=Multiply(A=@n29, B=@n30))) {name=simple bubble, nodeColor=<rgba 0 1 0.5313 1>}
  n31 = Fresnel(ExponentIn=Constant() {r=0.5})
  NamedRerouteDeclaration(None=LinearInterpolate(A=Constant() {r=0.2}, B=Constant() {r=0.5}, Alpha=OneMinus(None=@n31))) {name=RedColorRIM, nodeColor=<rgba 1 0.4922 0 1>}
  n32 = GetMaterialAttributes(None=MaterialFunctionCall(
    PreskinnedPosition=@n1.Pre skinned Local Position
    Gmap=@n1.Pre skinned Local Position
    MaterialAttributes=@n1.Pre skinned Local Position
    AO=@n1.Pre skinned Local Position
   ) {materialFunction=&CharacterSkinFX}.Result) {attributeGetTypes [
    #1
    #2
    #3
    #4
    #5
    #6
   ]}
  NamedRerouteDeclaration(None=@n32.Normal) {name=normal, nodeColor=<rgba 0 1 0.0156 1>}
  NamedRerouteDeclaration(None=@n32.Base Color) {name=Base Color, nodeColor=<rgba 0 0.8125 1 1>}
  NamedRerouteDeclaration(None=@n32.Curvature or Metallic) {name=Metallic, nodeColor=<rgba 0 0.8125 1 1>}
  NamedRerouteDeclaration(None=@n32.Specular) {name=Spec, nodeColor=<rgba 0 0.8125 1 1>}
  NamedRerouteDeclaration(None=@n32.Roughness) {name=Roughness, nodeColor=<rgba 0 0.8125 1 1>}
  NamedRerouteDeclaration(None=@n32.Emissive Color) {name=Emissive, nodeColor=<rgba 0 0.8125 1 1>}
  n33 = NamedRerouteUsage()
  n34 = Constant()
  n35 = Multiply(A=@n3.RGB, B=Constant() {r=0.45}) {constB=2}
  n36 = Power(Base=@n2.R)
  NamedRerouteDeclaration(None=Add(A=QualitySwitch(Default=@n33, Low=@n34, High=@n33, Medium=@n34, Epic=@n33), B=Multiply(A=@n35, B=@n36))) {name=bubble, nodeColor=<rgba 1 0.5859 0 1>}
  VertexInterpolator(VS=PreSkinnedPosition())
  VertexInterpolator(VS=PreSkinnedNormal())
 ]
}
